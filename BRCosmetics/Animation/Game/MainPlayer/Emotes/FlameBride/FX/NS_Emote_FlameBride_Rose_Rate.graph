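# Niagara effect for the FlameBride emote that throws out a short spray of tumbling red rose meshes that trail fire and embers.
# Rose spawns by burst and decaying rate, inherits velocity, and animates material parameters over life.
# Fire flipbook sprites and glowing ember particles spawn from Rose particles through particle reads, while Dirt adds dark falling specks independently.
# Time-of-day functions adjust color and brightness, and User.bIsFrontEndPreview switches fire and glow looks for front-end preview.
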NiagaraSystem /BRCosmetics/Animation/Game/MainPlayer/Emotes/FlameBride/FX/NS_Emote_FlameBride_Rose_Rate
dependencies {
 usedDataInterfaces [
  &NiagaraDataInterfaceCurve
  &NiagaraDataInterfaceParticleRead
  &NiagaraDataInterfaceColorCurve
  &NiagaraDataInterfaceSpriteRendererInfo
 ]
 usedDynamicInputs []
 usedModules [&SystemState]
}
emitters {
 Dirt {
  emitterUpdateScript [
   &EmitterState(Life Cycle Mode=::NewEnumerator1, Loop Behavior=::NewEnumerator1, UseLoopDelay=true)
   &SpawnBurst_Instantaneous(Spawn Count=5, Use Spawn Probability=false, UseLoopCountLimit=false)
   &SpawnRate(
    SpawnRate=&FloatFromCurve(
     FloatCurve=DataInterface() {propertyValues {Curve {keys [<at 0 1>,<at 0.2 1>,<at 0.6 0>,<at 1 0>]}}}
     CurveIndex=$Emitter.Age
     Scale Curve=40
    )
    Use Spawn Probability=false
   )
  ]
  particleSpawnScript [
   &InitializeParticle(
    Color=<rgba 0 0 0 1.5>
    Lifetime Max=3
    Lifetime Mode=::NewEnumerator1
    Sprite Rotation Mode=::NewEnumerator2
    Sprite Size Max=<1.2 2.5>
    Sprite Size Min=<0.6 1.5>
    Sprite Size Mode=::NewEnumerator2
    UsePositionOffset=false
    Write Material Random=true
   )
   &ShapeLocation(MaskSpawnGroup=false, Offset=<0 0 10>, Offset Mode=::NewEnumerator0, Override Seed=false, Sphere Radius=6, UseOwnerScale=true)
   &AddVelocity(Module.Use Falloff Distance=false, Override Seed=false, Velocity Mode=::NewEnumerator1)
  ]
  particleUpdateScript [
   &ParticleState()
   &ScaleSpriteSize(
    Uniform Curve Scale=&FloatFromCurve(FloatCurve=DataInterface() {propertyValues {Curve {keys [<at 0 0>,<at 0.17 0>,<at 0.2 1>]}}}, CurveIndex=$Emitter.Age)
    Uniform Curve Sprite Scale=DataInterface() {propertyValues {Curve {keys [<at 0 0 RCIM_Cubic>,<at 0.2 1 RCIM_Cubic>,<at 1 0 RCIM_Cubic>]}}}
   )
   &ScaleColor(
    Scale Alpha=&Multiply_Float(
     A=1
     B=&FloatFromCurve(
      FloatCurve=DataInterface() {propertyValues {Curve {keys [
          {arriveTangent=0.8687, interpMode=RCIM_Cubic, leaveTangent=0.8687, tangentMode=RCTM_User, time=0.1, value=0}
          <at 0.5 1 RCIM_Cubic>
          <at 1 0 RCIM_Cubic>
         ]}}}
      Scale Curve=&RandomRangeFloat(Minimum=0.5, Maximum=1)
     )
    )
    ScaleA=true
    ScaleRGB=true
   )
   &CurlNoiseForce(Bias Noise=false, Noise Strength=50, Pan Noise=false)
   &Drag(Drag=1.5, Use Linear Drag=true, Use Rotational Drag=true)
   &GravityForce(Gravity=&RandomRangeVector(Minimum=<0 0 -30>, Maximum=<0 0 -100>))
   &SolveForcesAndVelocity(Clamp Velocity=false, Limit Acceleration=false)
  ]
  rendererClasses [&NiagaraSpriteRendererProperties]
  renderers [{rendererClass=&NiagaraSpriteRendererProperties}]
 }
 Fire {
  emitterUpdateScript [
   &EmitterState(
    Life Cycle Mode=::NewEnumerator1
    Loop Behavior=::NewEnumerator1
    Loop Delay=0.6
    Loop Duration=&RandomRangeFloat(Minimum=0.6, Maximum=0.8)
    UseLoopDelay=true
   )
   &SpawnParticlesFromOtherEmitter(
    Attribute Reader=DataInterface() {propertyValues {EmitterBinding {bindingMode=Other, emitterName=Rose}}}
    Spawn Rate=15
    Use Maximum Age=false
    Use Minimum Age=false
   )
  ]
  particleSpawnScript [
   &InitializeParticle(
    Lifetime Max=1.3
    Lifetime Min=0.4
    Lifetime Mode=::NewEnumerator1
    Sprite Rotation Mode=::NewEnumerator2
    Sprite Size Mode=::NewEnumerator1
    Uniform Sprite Size Max=20
    UsePositionOffset=false
    Write Material Random=true
   )
   &SampleParticlesFromOtherEmitter(Velocity Sampling=::NewEnumerator0)
  ]
  particleUpdateScript [
   &ParticleState()
   &SpriteRotationRate(Rotation Rate=&RandomRangeFloat(Minimum=-100, Maximum=100))
   &ScaleSpriteSize(
    Uniform Curve Scale=1.3
    Uniform Curve Sprite Scale=DataInterface() {propertyValues {Curve {keys [
        {arriveTangent=2.2767, interpMode=RCIM_Cubic, leaveTangent=2.2767, tangentMode=RCTM_User, time=0, value=0.2}
        {arriveTangent=-0.009, interpMode=RCIM_Cubic, leaveTangent=-0.009, tangentMode=RCTM_User, time=1, value=1}
       ]}}}
   )
   &Drag(Use Linear Drag=true, Use Rotational Drag=true)
   &GravityForce(Gravity=&RandomRangeVector(Minimum=<0 0 0>, Maximum=<0 0 -10>)) {enabled=false}
   &CurlNoiseForce(Bias Noise=false, Noise Strength=100, Pan Noise=false) {enabled=false}
   &SolveForcesAndVelocity(Clamp Velocity=false, Limit Acceleration=false)
   &ScaleColor(
    Scale Alpha=&FloatFromCurve(
     FloatCurve=DataInterface() {propertyValues {Curve {keys [
         <at 0 0 RCIM_Cubic>
         {arriveTangent=-0.3533, interpMode=RCIM_Cubic, leaveTangent=-0.3533, tangentMode=RCTM_User, time=0.22, value=1}
         {arriveTangent=-1.4952, interpMode=RCIM_Cubic, leaveTangent=-1.4952, tangentMode=RCTM_User, time=0.8, value=0.5}
         <at 1 0 RCIM_Cubic>
        ]}}}
     Scale Curve=&MakeCustomFloatFromBool(Bool=$User.bIsFrontEndPreview, True Float=0.7, False Float=&RandomRangeFloat(Minimum=0.5, Maximum=1))
    )
    Scale RGB=&MakeCustomVectorFromBool(
     Bool=$User.bIsFrontEndPreview
     True Vector=&MakeVectorFromLinearColorRGB(Color=&ColorFromCurve(ColorCurve=DataInterface() {propertyValues {
         AlphaCurve {keys []}
         BlueCurve {keys [<at 0.0717 0.0133>, {arriveTangent=-0.0287, interpMode=RCIM_Cubic, leaveTangent=-0.0287, time=0.4137, value=0.01}, <at 0.536 0>]}
         GreenCurve {keys [<at 0.0717 0.22>, {arriveTangent=-0.4631, interpMode=RCIM_Cubic, leaveTangent=-0.4631, time=0.4137, value=0.075}, <at 0.536 0.005>]}
         RedCurve {keys [<at 0.0717 1>, {arriveTangent=-2.0644, interpMode=RCIM_Cubic, leaveTangent=-2.0644, time=0.4137, value=0.5}, <at 0.536 0.0417>]}
        }}))
     False Vector=&MakeVectorFromLinearColorRGB(Color=&ColorFromCurve(ColorCurve=DataInterface() {propertyValues {
         AlphaCurve {keys []}
         BlueCurve {keys [<at 0.0392 0.0243>,<at 0.1766 0 RCIM_Cubic>,<at 0.5058 0>]}
         GreenCurve {keys [<at 0.0392 0.3>,<at 0.1766 0.12 RCIM_Cubic>,<at 0.5058 0.005>]}
         RedCurve {keys [<at 0.0392 1.5>,<at 0.1766 1 RCIM_Cubic>,<at 0.5058 0.0417>]}
        }}))
    )
    ScaleA=true
    ScaleRGB=true
   )
   &ScaleColor(
    Scale Alpha=&FNTimeOfDayScalar(Night Scale=0.6, Morning Scale=0.7, Evening Scale=0.7)
    Scale RGB=&MakeVectorFromLinearColorRGB(Color=&FNTimeOfDayColor(Night Color=<rgba 1 0.9948 0.2694 1>, Morning Color=<rgba 1 1 1 1>, Day Color=<rgba 1 1 1 1>, Evening Color=<rgba 1 1 1 1>))
    ScaleA=true
    ScaleRGB=true
   ) {module=ScaleColor001}
   &SubUVAnimation(
    End Frame Range Override:NiagaraInt32=&RandomRangeInt(Minimum=12, Maximum=16)
    "Sprite Renderer "=DataInterface() {propertyValues {SpriteRenderer=&self:Fire_0.NiagaraSpriteRendererProperties_2}}
    Start Frame Range Override=7
    UseEndFrame=true
    UseStartFrame=true
   )
  ]
  rendererClasses [&NiagaraSpriteRendererProperties]
  renderers [{rendererClass=&NiagaraSpriteRendererProperties}]
 }
 Rose {
  emitterUpdateScript [
   &EmitterState(Life Cycle Mode=::NewEnumerator1, Loop Behavior=::NewEnumerator1, UseLoopDelay=false)
   &SpawnRate(
    SpawnRate=&FloatFromCurve(
     FloatCurve=DataInterface() {propertyValues {Curve {keys [<at 0 2>,<at 0.2 0.75>,<at 0.46 0>,<at 1 0>]}}}
     CurveIndex=$Emitter.Age
     Scale Curve=40
    )
    Use Spawn Probability=false
   )
   &SpawnBurst_Instantaneous(Spawn Count=3, Use Spawn Probability=false, UseLoopCountLimit=false)
  ]
  particleSpawnScript [
   &InitializeParticle(
    Color Mode=::NewEnumerator0
    Lifetime Mode=::NewEnumerator1
    Mesh Scale Max=<1 1 1>
    Mesh Scale Min=<0.3 0.3 0.3>
    Mesh Scale Mode=::NewEnumerator2
    UsePositionOffset=false
    Write Material Random=true
   )
   &InitialMeshOrientation(
    Mesh Orientation Mode=::NewEnumerator2
    Rotation=&RandomRangeVector(Minimum=<0 0 0>, Maximum=<1 1 1>)
    Use Rotation Vector=true
   )
   &ShapeLocation(MaskSpawnGroup=false, Override Seed=false, Sphere Radius=5, UseOwnerScale=true)
   &AddVelocity(Module.Use Falloff Distance=false, Override Seed=false, Velocity Mode=::NewEnumerator1, Velocity Speed=20) {module=AddVelocity001}
   &AddVelocity(Velocity=<-80 0 0>)
   &InheritVelocity(Inherited Velocity Speed Limit=20, Speed Limit=true, UseSpeedThreshold=false)
  ]
  particleUpdateScript [
   &ParticleState()
   &UpdateMeshOrientation(Rotation Rate=&RandomRangeFloat(Minimum=0.8, Maximum=1.2), Rotation Vector=<1 0.2 0.4>)
   &CurlNoiseForce(Bias Noise=false, Noise Strength=60, Pan Noise=false)
   &ScaleColor(
    Scale Alpha=&FloatFromCurve(FloatCurve=DataInterface() {propertyValues {Curve {keys [<at 0.8 1>,<at 1 1>]}}})
    Scale RGB=&MakeVectorFromLinearColorRGB(Color=&FNTimeOfDayColor(
      Night Color=<rgba 0.45 0.0302 0.027 3>
      Morning Color=<rgba 0.7292 0.0489 0.0438 3>
      Day Color=<rgba 0.7292 0.0489 0.0438 3>
      Evening Color=<rgba 0.7292 0.0489 0.0438 3>
     ))
    ScaleA=true
    ScaleRGB=true
   )
   &Drag(
    Drag=&FloatFromCurve(FloatCurve=DataInterface() {propertyValues {Curve {keys [<at 0.5 2>,<at 1 3>]}}})
    Use Linear Drag=true
    Use Rotational Drag=true
   )
   &GravityForce(Gravity=&RandomRangeVector(Minimum=<0 0 -100>, Maximum=<0 0 -200>))
   &SolveForcesAndVelocity(Clamp Velocity=false, Limit Acceleration=false)
   &DynamicMaterialParameters(
    Index 0 Param 1=&FloatFromCurve(
     FloatCurve=DataInterface() {propertyValues {Curve {keys [<at 0.3 1>,<at 1 -0.3>]}}}
     Scale Curve=&RandomRangeFloat(Minimum=0.8, Maximum=1)
    )
    Index 0 Param 2=&RandomRangeFloat(Minimum=0, Maximum=0.5)
    Index 0 Param 3=&RandomRangeFloat(Minimum=0, Maximum=0.5)
    Index 0 Param 4=&RandomRangeFloat(Minimum=0.5, Maximum=2)
    Module.Index0Float0Write=true
    Module.Index0Float1Write=true
    Module.Index0Float2Write=true
    Module.Index0Float3Write=true
   )
  ]
  rendererClasses [&NiagaraMeshRendererProperties]
  renderers [{rendererClass=&NiagaraMeshRendererProperties, rendererIndex=0}]
 }
 particle {
  emitterUpdateScript [
   &EmitterState(Life Cycle Mode=::NewEnumerator1, Loop Behavior=::NewEnumerator1, Loop Delay=0.2, Loop Duration=0.3, UseLoopDelay=true)
   &SpawnParticlesFromOtherEmitter(
    Attribute Reader=DataInterface() {propertyValues {EmitterBinding {bindingMode=Other, emitterName=Rose}}}
    Spawn Rate=8
    Use Maximum Age=false
    Use Minimum Age=false
   )
  ]
  particleSpawnScript [
   &InitializeParticle(
    Color=<rgba 1.2 1.2 1.2 1>
    Lifetime Max=3.5
    Lifetime Min=0.4
    Lifetime Mode=::NewEnumerator1
    Sprite Size Max=<1.2 2>
    Sprite Size Min=<0.6 1.5>
    Sprite Size Mode=::NewEnumerator2
    UsePositionOffset=false
    Write Material Random=true
   )
   &SampleParticlesFromOtherEmitter()
   &ShapeLocation(MaskSpawnGroup=false, Offset Mode=::NewEnumerator0, Override Seed=false, Sphere Radius=5, UseOwnerScale=true)
  ]
  particleUpdateScript [
   &ParticleState()
   &ScaleSpriteSize(
    Uniform Curve Scale=&FloatFromCurve(FloatCurve=DataInterface() {propertyValues {Curve {keys [<at 0 0>,<at 0.17 0>,<at 0.2 1.2>]}}}, CurveIndex=$Emitter.Age)
    Uniform Curve Sprite Scale=DataInterface() {propertyValues {Curve {keys [<at 0 0 RCIM_Cubic>,<at 0.2 1 RCIM_Cubic>,<at 1 0 RCIM_Cubic>]}}}
   )
   &ScaleColor(
    Scale Alpha=&Multiply_Float(
     A=1
     B=&FloatFromCurve(FloatCurve=DataInterface() {propertyValues {Curve {keys [
          {arriveTangent=0.8687, interpMode=RCIM_Cubic, leaveTangent=0.8687, tangentMode=RCTM_User, time=0.1, value=0}
          {arriveTangent=0.0152, interpMode=RCIM_Cubic, leaveTangent=0.0152, tangentMode=RCTM_User, time=0.4, value=1}
          <at 0.8 0 RCIM_Cubic>
         ]}}})
    )
    Scale RGB=&MakeVectorFromLinearColorRGB(Color=&FNTimeOfDayColor(
      Night Color=<rgba 1.2 0.3933 0.0198 1>
      Morning Color=<rgba 1.2 0.3484 0.023 1>
      Day Color=<rgba 1.2 0.2 0 1>
      Evening Color=<rgba 1.5 0.25 0 1>
     ))
    ScaleA=true
    ScaleRGB=true
   )
   /SetVariables(Particles.Glow03_Size=&Multiply_Vector2D_ByFloat(
     Float=&FloatFromCurve(
      FloatCurve=DataInterface() {propertyValues {Curve {keys [
          <at 0 0 RCIM_Cubic>
          <at 0.2 1.3 RCIM_Cubic>
          {arriveTangent=-0.1423, interpMode=RCIM_Cubic, leaveTangent=-0.1423, tangentMode=RCTM_User, time=1, value=0}
         ]}}}
      Scale Curve=10
     )
     Vector2D=<0.75 1>
    )) {isSetParameters=true}
   /SetVariables(Particles.Glow03_Color=&MakeLinearColorFromVectorAndFloat(
     "Vector (RGB)"=&MakeCustomVectorFromBool(
      Bool=$User.bIsFrontEndPreview
      True Vector=&MakeVectorFromLinearColorRGB(Color=<rgba 1 0.28 0.02 1>)
      False Vector=&MakeVectorFromLinearColorRGB(Color=&MultiplyLinearColor(
        Linear Color=&FNTimeOfDayColor(Night Color=<rgba 1 0.3167 0.0354 1>, Morning Color=<rgba 1 0.15 0.02 1>, Day Color=<rgba 1 0.15 0.02 1>, Evening Color=<rgba 1 0.15 0.02 1>)
        Scale Factor=&MakeVector4_FromFloat(Float=&Multiply_Float(A=1, B=&FNTimeOfDayScalar(Night Scale=0.5, Morning Scale=0.5, Evening Scale=0.5)))
       ))
     )
     "Float (Alpha)"=&Multiply_Float(
      A=0.1
      B=&FloatFromCurve(
       FloatCurve=DataInterface() {propertyValues {Curve {keys [<at 0 0>,<at 0.15 0>,<at 0.25 1>]}}}
       CurveIndex=$Emitter.Age
       Scale Curve=&MakeCustomFloatFromBool(Bool=$User.bIsFrontEndPreview, True Float=0.6, False Float=1)
      )
     )
    )) {isSetParameters=true}
   &CurlNoiseForce(Bias Noise=false, Noise Strength=100, Pan Noise=false)
   &Drag(Drag=2, Use Linear Drag=true, Use Rotational Drag=true)
   &GravityForce(Gravity=&RandomRangeVector(Minimum=<0 0 -20>, Maximum=<0 0 -180>))
   &SolveForcesAndVelocity(Clamp Velocity=false, Limit Acceleration=false)
   &ScaleSpriteSizeBySpeed(Max Scale Factor=<1 2>)
  ]
  rendererClasses [&NiagaraSpriteRendererProperties]
  renderers [
   {rendererClass=&NiagaraSpriteRendererProperties}
   {rendererClass=&NiagaraSpriteRendererProperties, rendererIndex=1}
  ]
 }
}
summary {
 systemName=NS_Emote_FlameBride_Rose_Rate
 emitters [
  {emitterName=Rose, rendererClasses [&NiagaraMeshRendererProperties]}
  {emitterName=particle, rendererClasses [&NiagaraSpriteRendererProperties]}
  {emitterName=Dirt, rendererClasses [&NiagaraSpriteRendererProperties]}
  {emitterName=Fire, rendererClasses [&NiagaraSpriteRendererProperties]}
 ]
 userVariables [User.bIsFrontEndPreview {type=NiagaraBool}]
}
